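# Physics asset for the back hair of the QuicheLorraineLime medium female character, defining bodies and joints for dynamic hair motion.
# Center, left and right bone chains hang from small base spheres locked angularly to the head, with angle-limited joints between segments.
# Upper side segments use three parallel capsules each to approximate wide locks.
# Unconstrained torso, neck, clavicle, chest and upper-arm capsules supply body collision shapes, with greatly elongated spine and lower-neck capsules.
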PhysicsAsset /BRCosmetics/Characters/Player/Female/Medium/Bodies/F_MED_QuicheLorraineLime_Hair/Meshes/Parts/F_MED_QuicheLorraineLime_BackHair_Physics
bodies {
 clavicle_l [{
   shapeType=Capsule
   length=7.3238
   radius=5.426
   shapeName=clavicle_l_capsule
   center=<5.1656 0.7351 1.3104>
   rotation=<rot -0.0002 -94.7719 101.5585>
  }]
 clavicle_r [{
   shapeType=Capsule
   length=7.3238
   radius=5.426
   shapeName=clavicle_l_capsule
   center=<-6.0597 -0.5752 -1.5348>
   rotation=<rot -0.0988 -96.3072 98.4075>
  }]
 deform_pec_aim_l [{shapeType=Capsule, length=7.8944, radius=4.2789, shapeName=deform_pec_aim_l_capsule, center=<7.7881 0 0>, rotation=<rot 0 -90 90>}]
 deform_pec_aim_r [{shapeType=Capsule, length=7.8944, radius=4.2789, shapeName=deform_pec_aim_r_capsule, center=<-7.7881 0 0>, rotation=<rot 0 -90 90>}]
 dyn_hair_ba_ct_03 [{shapeType=Capsule, length=10.043, radius=1.741, shapeName=dyn_hair_l_02_capsule, center=<5.0795 0 0>, rotation=<rot 90 0 0>}]
 dyn_hair_ba_ct_03_base [{shapeType=Sphere, radius=0.505, shapeName=dyn_hair_ba_ct_03_base_sphere}]
 dyn_hair_ba_ct_04 [{shapeType=Capsule, length=10.0346, radius=1.741, shapeName=dyn_hair_l_02_capsule, center=<5.0753 0 0>, rotation=<rot 90 0 0>}]
 dyn_hair_ba_ct_05 [{shapeType=Capsule, length=8.6192, radius=1.741, shapeName=dyn_hair_l_02_capsule, center=<4.3676 0 0>, rotation=<rot 90 0 0>}]
 dyn_hair_ba_ct_06 [{shapeType=Capsule, length=5.884, radius=1.741, shapeName=dyn_hair_l_02_capsule, center=<3 0 0>, rotation=<rot 90 0 0>}]
 dyn_hair_ba_lf_03 [
  {shapeType=Capsule, length=10.5539, radius=1.5015, shapeName=dyn_hair_l_02_capsule, center=<5.0795 2.8434 0>, rotation=<rot 90 0 0>}
  {shapeType=Capsule, length=10.5539, radius=1.5015, shapeName=dyn_hair_l_02_capsule, center=<5.0795 0 0>, rotation=<rot 90 0 0>}
  {shapeType=Capsule, length=10.5539, radius=1.5015, shapeName=dyn_hair_l_02_capsule, center=<5.0795 -3.0968 0>, rotation=<rot 90 0 0>}
 ]
 dyn_hair_ba_lf_03_base [{shapeType=Sphere, radius=0.505, shapeName=dyn_hair_ba_lf_03_base_sphere}]
 dyn_hair_ba_lf_04 [
  {shapeType=Capsule, length=10.2741, radius=1.5015, shapeName=dyn_hair_l_02_capsule, center=<5.0753 0 0>, rotation=<rot 90 0 0>}
  {shapeType=Capsule, length=10.2741, radius=1.5015, shapeName=dyn_hair_l_02_capsule, center=<5.0753 2.8793 0>, rotation=<rot 90 0 0>}
  {shapeType=Capsule, length=10.2741, radius=1.5015, shapeName=dyn_hair_l_02_capsule, center=<5.0753 -3.2813 0>, rotation=<rot 90 0 0>}
 ]
 dyn_hair_ba_lf_05 [{shapeType=Capsule, length=6.1235, radius=1.5015, shapeName=dyn_hair_l_02_capsule, center=<3 0 0>, rotation=<rot 90 0 0>}]
 dyn_hair_ba_rt_03 [
  {shapeType=Capsule, length=10.0218, radius=1.7623, shapeName=dyn_hair_l_02_capsule, center=<5.0795 0 0>, rotation=<rot 90 0 0>}
  {shapeType=Capsule, length=10.0218, radius=1.7623, shapeName=dyn_hair_l_02_capsule, center=<5.0795 -2.8846 0>, rotation=<rot 90 0 0>}
  {shapeType=Capsule, length=10.0218, radius=1.7623, shapeName=dyn_hair_l_02_capsule, center=<5.0795 3.2634 0>, rotation=<rot 90 0 0>}
 ]
 dyn_hair_ba_rt_03_base [{shapeType=Sphere, radius=0.505, shapeName=dyn_hair_ba_rt_03_base_sphere}]
 dyn_hair_ba_rt_04 [
  {shapeType=Capsule, length=9.0134, radius=2.7623, shapeName=dyn_hair_l_02_capsule, center=<5.0753 -3.8741 0>, rotation=<rot 90 0 0>}
  {shapeType=Capsule, length=10.0134, radius=1.7623, shapeName=dyn_hair_l_02_capsule, center=<5.0753 0 0>, rotation=<rot 90 0 0>}
  {shapeType=Capsule, length=10.0134, radius=1.7623, shapeName=dyn_hair_l_02_capsule, center=<5.0753 3.2361 0>, rotation=<rot 90 0 0>}
 ]
 dyn_hair_ba_rt_05 [{shapeType=Capsule, length=5.8627, radius=1.7623, shapeName=dyn_hair_l_02_capsule, center=<3 0 0>, rotation=<rot 90 0 0>}]
 head [{shapeType=Capsule, length=8.1154, radius=5.701, shapeName=head_capsule, center=<1.8267 1.7211 -0.7388>, rotation=<rot -17.7283 -90 -89.4925>}]
 neck_01 [{shapeType=Capsule, length=145.1716, radius=3.75, shapeName=neck_01_capsule, center=<0.4485 0.567 0>, rotation=<rot 0 -90 180>}]
 neck_02 [{shapeType=Capsule, length=1.2606, radius=4.2139, shapeName=neck_02_capsule, center=<3.9212 0 0>, rotation=<rot 0 -90 90>}]
 spine_01 [{shapeType=Capsule, length=140.6805, radius=9.1696, shapeName=spine_05_capsule, center=<-2.4195 -1.8029 0>, rotation=<rot 0 -90 0>}]
 spine_02 [{shapeType=Capsule, length=141.0596, radius=8.7904, shapeName=spine_05_capsule, center=<0.2641 0.1017 0>, rotation=<rot 0 -90 0>}]
 spine_03 [{shapeType=Capsule, length=139.8144, radius=10.0357, shapeName=spine_05_capsule, center=<3.8652 2.0057 0>, rotation=<rot 0 -90 0>}]
 spine_04 [{shapeType=Capsule, length=140.4591, radius=9.391, shapeName=spine_05_capsule, center=<0.2641 1.4508 0>, rotation=<rot 0 -90 0>}]
 spine_05 [{shapeType=Capsule, length=142.3849, radius=7.4651, shapeName=spine_05_capsule, center=<0.2641 0.1017 0>, rotation=<rot 0 -90 0>}]
 upperarm_l [{shapeType=Capsule, length=47.6994, radius=4.3565, shapeName=upperarm_l_capsule, center=<17.656 1.3544 0>, rotation=<rot 0 -90 90>}]
 upperarm_r [{shapeType=Capsule, length=27.4253, radius=5.4234, shapeName=upperarm_r_capsule, center=<-15.0717 -1.4224 0>, rotation=<rot 0 -90 90>}]
}
constraints [
 {bone1Name=dyn_hair_ba_ct_03, bone2Name=dyn_hair_ba_ct_03_base, swing1LimitDegrees=20, swing2LimitDegrees=60, twistLimitDegrees=10}
 {bone1Name=dyn_hair_ba_ct_04, bone2Name=dyn_hair_ba_ct_03, swing1LimitDegrees=20, swing2LimitDegrees=60, twistLimitDegrees=10}
 {bone1Name=dyn_hair_ba_ct_05, bone2Name=dyn_hair_ba_ct_04, swing1LimitDegrees=20, swing2LimitDegrees=60, twistLimitDegrees=10}
 {bone1Name=dyn_hair_ba_ct_06, bone2Name=dyn_hair_ba_ct_05, swing1LimitDegrees=20, swing2LimitDegrees=60, twistLimitDegrees=10}
 {bone1Name=dyn_hair_ba_lf_03, bone2Name=dyn_hair_ba_lf_03_base, swing1LimitDegrees=20, swing2LimitDegrees=60, twistLimitDegrees=10}
 {bone1Name=dyn_hair_ba_lf_04, bone2Name=dyn_hair_ba_lf_03, swing1LimitDegrees=20, swing2LimitDegrees=60, twistLimitDegrees=10}
 {bone1Name=dyn_hair_ba_lf_05, bone2Name=dyn_hair_ba_lf_04, swing1LimitDegrees=20, swing2LimitDegrees=60, twistLimitDegrees=10}
 {bone1Name=dyn_hair_ba_rt_03, bone2Name=dyn_hair_ba_rt_03_base, swing1LimitDegrees=20, swing2LimitDegrees=60, twistLimitDegrees=10}
 {bone1Name=dyn_hair_ba_rt_04, bone2Name=dyn_hair_ba_rt_03, swing1LimitDegrees=20, swing2LimitDegrees=60, twistLimitDegrees=10}
 {bone1Name=dyn_hair_ba_rt_05, bone2Name=dyn_hair_ba_rt_04, swing1LimitDegrees=20, swing2LimitDegrees=60, twistLimitDegrees=10}
 {
  bone1Name=dyn_hair_ba_ct_03_base
  bone2Name=head
  swing1LimitDegrees=45
  swing1Motion=Locked
  swing2LimitDegrees=45
  swing2Motion=Locked
  twistLimitDegrees=45
  twistMotion=Locked
 }
 {
  bone1Name=dyn_hair_ba_lf_03_base
  bone2Name=head
  swing1LimitDegrees=45
  swing1Motion=Locked
  swing2LimitDegrees=45
  swing2Motion=Locked
  twistLimitDegrees=45
  twistMotion=Locked
 }
 {
  bone1Name=dyn_hair_ba_rt_03_base
  bone2Name=head
  swing1LimitDegrees=45
  swing1Motion=Locked
  swing2LimitDegrees=45
  swing2Motion=Locked
  twistLimitDegrees=45
  twistMotion=Locked
 }
]
