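# Physics asset for the front hair of the QuicheLorraineLime medium female character cosmetic, simulating two dynamic hair strands against the upper body.
# Capsules on the head, neck, spine, clavicles, pecs and upper arms provide body collision shapes.
# A five-segment left chain and three-segment right chain hang from head-linked bases with free angular motion, then use limited swing and near-zero twist.
# Two left segments use triple side-by-side capsules for wider coverage.
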PhysicsAsset /BRCosmetics/Characters/Player/Female/Medium/Bodies/F_MED_QuicheLorraineLime_Hair/Meshes/Parts/F_MED_QuicheLorraineLime_Front_Hair_Physics
bodies {
 clavicle_l [{
   shapeType=Capsule
   length=7.8238
   radius=4.926
   shapeName=clavicle_l_capsule
   center=<5.0016 0.7488 0.5058>
   rotation=<rot -0.0002 -94.7719 101.5585>
  }]
 clavicle_r [{
   shapeType=Capsule
   length=7.3238
   radius=5.426
   shapeName=clavicle_l_capsule
   center=<-6.0597 -0.5752 -1.5348>
   rotation=<rot -0.0988 -96.3072 98.4075>
  }]
 deform_pec_aim_l [{shapeType=Capsule, length=7.8944, radius=4.2789, shapeName=deform_pec_aim_l_capsule, center=<7.7881 0 0>, rotation=<rot 0 -90 90>}]
 deform_pec_aim_r [{shapeType=Capsule, length=7.8944, radius=4.2789, shapeName=deform_pec_aim_r_capsule, center=<-7.7881 0 0>, rotation=<rot 0 -90 90>}]
 dyn_hair_fr_lf_01 [{shapeType=Capsule, length=4.3949, radius=2.2866, shapeName=dyn_hair_l_01_capsule, center=<3.7008 0 0>, rotation=<rot 90 0 0>}]
 dyn_hair_fr_lf_02 [
  {shapeType=Capsule, length=7.0389, radius=2.2866, shapeName=dyn_hair_l_02_capsule, center=<4.3502 0 0.2854>, rotation=<rot 90 0 0>}
  {shapeType=Capsule, length=8.0389, radius=1.2866, shapeName=dyn_hair_l_02_capsule, center=<4.3502 -3.1313 -0.7756>, rotation=<rot 90 0 0>}
  {shapeType=Capsule, length=8.0389, radius=1.2866, shapeName=dyn_hair_l_02_capsule, center=<4.3502 2.8676 -1.0942>, rotation=<rot 90 0 0>}
 ]
 dyn_hair_fr_lf_03 [
  {shapeType=Capsule, length=6.5244, radius=2.2866, shapeName=dyn_hair_l_02_capsule, center=<3.593 -3.3866 0.3>, rotation=<rot 90 0 0>}
  {shapeType=Capsule, length=6.5244, radius=2.2866, shapeName=dyn_hair_l_02_capsule, center=<3.593 0 0.3>, rotation=<rot 90 0 0>}
  {shapeType=Capsule, length=6.5244, radius=2.2866, shapeName=dyn_hair_l_02_capsule, center=<3.593 3.0986 0.3>, rotation=<rot 90 0 0>}
 ]
 dyn_hair_fr_lf_04 [{shapeType=Capsule, length=3.5762, radius=3.2866, shapeName=dyn_hair_l_02_capsule, center=<3.6189 -0.7412 0.33>, rotation=<rot 90 0 0>}]
 dyn_hair_fr_lf_05 [{shapeType=Capsule, length=6.0762, radius=2.7866, shapeName=dyn_hair_l_02_capsule, center=<3.6189 0 -0.7686>, rotation=<rot 90 0 0>}]
 dyn_hair_fr_lf_base [{shapeType=Capsule, length=1.2615, radius=1.2866, shapeName=dyn_hair_l_MSTR_capsule, rotation=<rot 90 0 0>}]
 dyn_hair_fr_rt_01 [{shapeType=Capsule, length=7.9639, radius=1.457, shapeName=dyn_hair_l_01_capsule, center=<3.8979 0 0>, rotation=<rot 90 0 0>}]
 dyn_hair_fr_rt_02 [{shapeType=Capsule, length=7.4338, radius=1.457, shapeName=dyn_hair_l_02_capsule, center=<3.6329 0 0>, rotation=<rot 90 0 0>}]
 dyn_hair_fr_rt_03 [{shapeType=Capsule, length=4.168, radius=3.7566, shapeName=dyn_hair_l_02_capsule, center=<2.5178 0 -0.0606>, rotation=<rot 73.6081 180 -180>}]
 dyn_hair_fr_rt_base [{shapeType=Capsule, length=1.2615, radius=1.457, shapeName=dyn_hair_l_MSTR_capsule, rotation=<rot 90 0 0>}]
 head [{shapeType=Capsule, length=8.1154, radius=5.701, shapeName=head_capsule, center=<0.7935 4.6689 0.2132>, rotation=<rot -17.7283 -90 -89.4925>}]
 neck_01 [{shapeType=Capsule, length=159.0221, radius=4.0001, shapeName=neck_01_capsule, center=<0.4485 0 0>, rotation=<rot 0 -90 -180>}]
 neck_02 [{shapeType=Capsule, length=4, radius=6.7139, shapeName=neck_02_capsule, center=<-0.3674 0 0>, rotation=<rot 0 -90 90>}]
 spine_01 [{shapeType=Capsule, length=21.6802, radius=10.1696, shapeName=spine_05_capsule, center=<-2.4195 -1.8029 0>, rotation=<rot 0 -90 0>}]
 spine_02 [{shapeType=Capsule, length=13.5593, radius=10.7904, shapeName=spine_05_capsule, center=<0.2641 0.1017 0>, rotation=<rot 0 -90 0>}]
 spine_03 [{shapeType=Capsule, length=11.3141, radius=10.0357, shapeName=spine_05_capsule, center=<3.8652 2.0057 0>, rotation=<rot 0 -90 0>}]
 spine_05 [{shapeType=Capsule, length=19.3846, radius=7.4651, shapeName=spine_05_capsule, center=<0.2641 0.1017 0>, rotation=<rot 0 -90 0>}]
 upperarm_l [{shapeType=Capsule, length=35.1994, radius=4.3565, shapeName=upperarm_l_capsule, center=<19.6395 1.3544 0>, rotation=<rot 0 -90 90>}]
 upperarm_r [{shapeType=Capsule, length=42.9253, radius=3.9234, shapeName=upperarm_r_capsule, center=<-15.0717 -1.4224 0>, rotation=<rot 0 -90 90>}]
}
constraints [
 {
  bone1Name=dyn_hair_fr_lf_base
  bone2Name=head
  swing1LimitDegrees=45
  swing1Motion=Free
  swing2LimitDegrees=45
  swing2Motion=Free
  twistLimitDegrees=45
  twistMotion=Free
 }
 {bone1Name=dyn_hair_fr_lf_01, bone2Name=dyn_hair_fr_lf_base, swing1LimitDegrees=20, swing2LimitDegrees=30, twistLimitDegrees=0}
 {bone1Name=dyn_hair_fr_lf_02, bone2Name=dyn_hair_fr_lf_01, swing1LimitDegrees=40, swing2LimitDegrees=60, twistLimitDegrees=1}
 {bone1Name=dyn_hair_fr_lf_03, bone2Name=dyn_hair_fr_lf_02, swing1LimitDegrees=20, swing2LimitDegrees=60, twistLimitDegrees=1}
 {bone1Name=dyn_hair_fr_lf_04, bone2Name=dyn_hair_fr_lf_03, swing1LimitDegrees=20, swing2LimitDegrees=60, twistLimitDegrees=1}
 {bone1Name=dyn_hair_fr_lf_05, bone2Name=dyn_hair_fr_lf_04, swing1LimitDegrees=20, swing2LimitDegrees=60, twistLimitDegrees=1}
 {
  bone1Name=dyn_hair_fr_rt_base
  bone2Name=head
  swing1LimitDegrees=45
  swing1Motion=Free
  swing2LimitDegrees=45
  swing2Motion=Free
  twistLimitDegrees=45
  twistMotion=Free
 }
 {bone1Name=dyn_hair_fr_rt_01, bone2Name=dyn_hair_fr_rt_base, swing1LimitDegrees=11.25, swing2LimitDegrees=22.5, twistLimitDegrees=0}
 {bone1Name=dyn_hair_fr_rt_02, bone2Name=dyn_hair_fr_rt_01, swing1LimitDegrees=12, swing2LimitDegrees=27.5, twistLimitDegrees=0}
 {bone1Name=dyn_hair_fr_rt_03, bone2Name=dyn_hair_fr_rt_02, swing1LimitDegrees=12, swing2LimitDegrees=30, twistLimitDegrees=0}
]
