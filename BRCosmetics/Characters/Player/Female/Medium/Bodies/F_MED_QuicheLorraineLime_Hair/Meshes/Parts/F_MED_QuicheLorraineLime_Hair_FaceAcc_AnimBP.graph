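# Post-copy animation blueprint for the hair and face accessory part of the QuicheLorraineLime female character, adding secondary hair motion.
# The pose copied from the attached parent runs through a Control Rig, then chained back- and front-hair RigidBody simulations in spine_05 space with overridden gravity.
# Customization load disables back-hair simulation when backbling is present.
# Movement events nearly disable front-hair simulation during tactical sprint, rechecking it on jumps and falls.
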AnimBlueprint /BRCosmetics/Characters/Player/Female/Medium/Bodies/F_MED_QuicheLorraineLime_Hair/Meshes/Parts/F_MED_QuicheLorraineLime_Hair_FaceAcc_AnimBP
target_skeleton=&/F_MED_QuicheLorraineLime_Hair_FaceAcc_Skeleton
graphs [
 AnimGraph {nodes [
   n1 = Control Rig(Source=Copy Pose From Mesh() {bUseAttachedParent=true}) {controlRigAssetReference {blueprintRigClass=&/F_MED_QuicheLorraineLime_Hair_Spl_CtrlRig.F_MED_QuicheLorraineLime_Hair_Spl_CtrlRig_C}}
   n2 = RigidBody!(
    ComponentPose=Local To Component(LocalPose=@n1)
    Alpha=Get hair back alpha() -> (hair back alpha:double)
    OverrideWorldGravity:Vector=Get Gravity Override()
    ExternalForce
   ) -> (Pose) {
    bClampLinearTranslationLimitToRefPose=true
    bOverrideWorldGravity=true
    overridePhysicsAsset=&/F_MED_QuicheLorraineLime_BackHair_Physics
    simulationSpace=BaseBoneSpace
    baseBoneRef {boneName=spine_05}
    componentAppliedLinearAccClamp=<3000 3000 3000>
    overrideWorldGravity {z=-980}
    simSpaceSettings {
     dampingAlpha=0
     maxAngularAcceleration=40
     maxAngularVelocity=8
     maxLinearAcceleration=1500
     maxLinearVelocity=700
     worldAlpha=1
     externalLinearDragV=<0.2 0.2 0.2>
    }
   }
   n3 = RigidBody(ComponentPose=@n2, Alpha=Get Hair Front Alpha() -> (Hair Front Alpha:float), OverrideWorldGravity=Get Gravity Override()) {
    bClampLinearTranslationLimitToRefPose=true
    bOverrideWorldGravity=true
    overridePhysicsAsset=&/F_MED_QuicheLorraineLime_Front_Hair_Physics
    simulationSpace=BaseBoneSpace
    baseBoneRef {boneName=spine_05}
    componentAppliedLinearAccClamp=<3000 3000 3000>
    overrideWorldGravity {z=-980}
    simSpaceSettings {
     dampingAlpha=0
     maxAngularAcceleration=40
     maxAngularVelocity=8
     maxLinearAcceleration=1500
     maxLinearVelocity=700
     worldAlpha=1
     externalLinearDragV=<0.2 0.2 0.2>
    }
   }
   Output Pose(Result=Component To Local(ComponentPose=@n3)) {layerGroup=None}
  ]}
 EventGraph {nodes [
   Event BlueprintUpdateAnimation()
   > n1 = Gravity Override State Function DEPRECATED(CharacterPartAnimInstance=Self-Reference(), GameTimeInSeconds=GetGameTimeInSeconds(), GravityOverrideParams=Get Gravity Override Params())
   > Set Gravity Override(Gravity Override=@n1)
   Event OnCharacterCustomizationLoaded()
   n2 = CheckForBackbling()
   > Set Has Backbling(Has Backbling:double="To Float (Boolean)"(InBool=@n2)) -> (Output_Get:double)
   > n3 = Branch(Condition=@n2)
   > Set hair back alpha(hair back alpha:double=0) -> (Output_Get:double)
   Set hair back alpha(execute=@n3.else, hair back alpha=0.667)
   n4 = Event OnBeginJumping()
   n5 = Branch(execute=[@n4.then, @n8.then], Condition=Get bIsTacticalSprinting())
   > Set Hair Front Alpha(Hair Front Alpha:float=0.05) -> (Output_Get:float)
   Set Hair Front Alpha(execute=@n5.else, Hair Front Alpha=1)
   n6 = Event OnEndJumping()
   n7 = Branch(execute=[@n6.then, @n9.then], Condition=Get bIsTacticalSprinting())
   > Set Hair Front Alpha(Hair Front Alpha=0.05)
   Set Hair Front Alpha(execute=[@n10.then, @n7.else], Hair Front Alpha=1)
   n8 = Event OnBeginFalling()
   n9 = Event OnEndFalling()
   Event OnBeginTacticalSprinting() {is=title K2Node_Event}
   > Set Hair Front Alpha(Hair Front Alpha=0.05)
   n10 = Event OnEndTacticalSprinting() {is=title K2Node_Event}
  ]}
]
variables [
 Gravity Override {type=Vector}
 Gravity Override Params {type=GravityOverrideParamsStruct}
 hair back alpha {type=number, default=0.667}
 Hair Front Alpha {type=number, default=1}
 Has Backbling {type=number}
 is Skyfloating {type=boolean}
 is Wearing Backbling {type=boolean}
 is Wearing Backbling or Cape {type=boolean}
 is Wearing Cape {type=boolean}
 isFallingNotSkydiving {type=boolean}
 isFrontEnd {type=boolean}
 Override World Gravity {type=Vector, default=<0 0 -980>}
 Sim Space Settings {
  type=SimSpaceSettings
  default {
   dampingAlpha=1
   maxAngularAcceleration=10000
   maxAngularVelocity=8
   maxLinearAcceleration=10000
   maxLinearVelocity=800
   velocityScaleZ=1
   worldAlpha=0.35
   externalLinearDragV=<0 0 0>
  }
 }
]
